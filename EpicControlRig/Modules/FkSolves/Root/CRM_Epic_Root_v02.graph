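# Epic modular Control Rig root module that spawns Global, Local and Root controls, with Local as an available space for Root.
# Construction publishes controls, global scale and colors as module metadata and auto-sizes controls from root-to-body distance.
# Forward solve optionally locks Global/Local scale, drives the root bone from Root, and draws channel-toggled debug path lines for Global and Local.
# Backwards solve snaps all three controls to the root bone.
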ControlRigBlueprint /EpicControlRig/Modules/FkSolves/Root/CRM_Epic_Root_v02
graphs [
 ConstructionGraph {nodes [
   RigUnit_PrepareForExecution = &RigUnit_PrepareForExecution()
   RigVMFunction_Sequence = RigVMAggregateNode(ExecuteContext=@RigUnit_PrepareForExecution.ExecutePin)
   RigVMVariableNode(
    Value=/Script/ControlRig.RigUnit_ResolveConnector(Connector=(Type=Connector,Name="RootJoint"), SkipSocket=False) {is=struct RigVMUnitNode}.Result
    ExecuteContext=@RigVMFunction_Sequence.A
   ) {variable=Root Bone}
   > RigVMVariableNode(Value=&RigUnit_ResolveConnector(Connector=(Type=Connector,Name="Body"), SkipSocket=False).Result) {variable=Body}
   SpawnControl = /Script/ControlRig.RigUnit_HierarchyAddControlTransform(
    InitialValue=(Rotation=(X=0,Y=0,Z=0,W=1),Translation=(X=0,Y=0,Z=0),Scale3D=(X=1,Y=1,Z=1))
    Name=Global
    OffsetSpace=LocalSpace
    OffsetTransform=(Rotation=(X=0,Y=0,Z=0,W=1),Translation=(X=0,Y=0,Z=0),Scale3D=(X=1,Y=1,Z=1))
    Parent=(Type=Bone,Name="None")
    Settings="(InitialSpace=LocalSpace,bUsePreferredRotationOrder=False,PreferredRotationOrder=YZX,Limits=(LimitTranslationX=(bMinimum=False,bMaximum=False),LimitTranslationY=(bMinimum=False,bMaximum=False),LimitTranslationZ=(bMinimum=False,bMaximum=False),LimitPitch=(b ...(truncated)"
    ExecutePin=@RigVMFunction_Sequence.B
   ) {
    is=struct RigVMUnitNode
    template="SpawnControl::Execute(in InitialValue,in Settings,in OffsetTransform,in OffsetSpace,in Parent,in Name,out Item)"
   }
   > RigVMVariableNode(Value=@SpawnControl.Item) {variable=Global Control}
   > Set Up Control Path = RigVMFunctionReferenceNode(Control=@SpawnControl.Item)
   > RigVMVariableNode(Value=@"Set Up Control Path".Vis Channel) {variable=Global Control Path Vis Channel}
   > RigVMVariableNode(Value=@"Set Up Control Path".Distance Channel) {variable=Global Control Path Distance Channel}
   > n1 = &RigUnit_HierarchyAddControlTransform(
    InitialValue=(Rotation=(X=0,Y=0,Z=0,W=1),Translation=(X=0,Y=0,Z=0),Scale3D=(X=1,Y=1,Z=1))
    Name=Local
    OffsetSpace=LocalSpace
    OffsetTransform=(Rotation=(X=0,Y=0,Z=0,W=1),Translation=(X=0,Y=0,Z=0),Scale3D=(X=1,Y=1,Z=1))
    Settings="(InitialSpace=LocalSpace,bUsePreferredRotationOrder=False,PreferredRotationOrder=YZX,Limits=(LimitTranslationX=(bMinimum=False,bMaximum=False),LimitTranslationY=(bMinimum=False,bMaximum=False),LimitTranslationZ=(bMinimum=False,bMaximum=False),LimitPitch=(b ...(truncated)"
    Parent=RigVMVariableNode() {variable=Global Control}.Value
   ) {template="SpawnControl::Execute(in InitialValue,in Settings,in OffsetTransform,in OffsetSpace,in Parent,in Name,out Item)"}
   > RigVMVariableNode(Value=@n1.Item) {variable=Local Control}
   > Set Up Control Path_1 = RigVMFunctionReferenceNode(Control=@n1.Item)
   > RigVMVariableNode(Value=@"Set Up Control Path_1".Vis Channel) {variable=Local Control Path Vis Channel}
   > RigVMVariableNode(Value=@"Set Up Control Path_1".Distance Channel) {variable=Local Control Path Distance Channel}
   > n2 = &RigUnit_HierarchyAddControlTransform(
    InitialValue=(Rotation=(X=0,Y=0,Z=0,W=1),Translation=(X=0,Y=0,Z=0),Scale3D=(X=1,Y=1,Z=1))
    Name=Root
    OffsetSpace=LocalSpace
    OffsetTransform=(Rotation=(X=0,Y=0,Z=0,W=1),Translation=(X=0,Y=0,Z=0),Scale3D=(X=1,Y=1,Z=1))
    Settings="(InitialSpace=LocalSpace,bUsePreferredRotationOrder=False,PreferredRotationOrder=YZX,Limits=(LimitTranslationX=(bMinimum=False,bMaximum=False),LimitTranslationY=(bMinimum=False,bMaximum=False),LimitTranslationZ=(bMinimum=False,bMaximum=False),LimitPitch=(b ...(truncated)"
    Parent=RigVMVariableNode() {variable=Global Control}.Value
   ) {template="SpawnControl::Execute(in InitialValue,in Settings,in OffsetTransform,in OffsetSpace,in Parent,in Name,out Item)"}
   > RigVMVariableNode(Value=@n2.Item) {variable=Root Control}
   > /Script/ControlRig.RigUnit_AddAvailableSpaces(
    Spaces=((Key=(Type=Control,Name="None"),Label="Local Control"))
    Control=@n2.Item
    Spaces.0.Key=RigVMVariableNode() {variable=Local Control}.Value
   ) {is=struct RigVMUnitNode}
   Set_Module_Metadata = RigVMDispatchNode(Name=Global Control, NameSpace=Self, Value=RigVMVariableNode() {variable=Global Control}.Value, ExecuteContext=@RigVMFunction_Sequence.C) {template="DISPATCH_RigDispatch_SetModuleMetadata(in Name,in NameSpace,in Value,out Success)"}
   > Set_Module_Metadata_2 = RigVMDispatchNode(Name=Root Control, NameSpace=Self, Value=RigVMVariableNode() {variable=Root Control}.Value) {template="DISPATCH_RigDispatch_SetModuleMetadata(in Name,in NameSpace,in Value,out Success)"}
   > Set_Module_Metadata_1 = RigVMDispatchNode(Name=Local Control, NameSpace=Self, Value=RigVMVariableNode() {variable=Local Control}.Value) {template="DISPATCH_RigDispatch_SetModuleMetadata(in Name,in NameSpace,in Value,out Success)"}
   > Set_Module_Metadata_2_1_1 = RigVMDispatchNode(Name=Global Control Scale, NameSpace=Self, Value=RigVMVariableNode() {variable=Global Control Scale}.Value) {template="DISPATCH_RigDispatch_SetModuleMetadata(in Name,in NameSpace,in Value,out Success)"}
   > Set_Module_Metadata_2_1_1_1 = RigVMDispatchNode(Name=Global Left Control Color, NameSpace=Self, Value=RigVMVariableNode() {variable=Global Left Control Color}.Value) {template="DISPATCH_RigDispatch_SetModuleMetadata(in Name,in NameSpace,in Value,out Success)"}
   > Set_Module_Metadata_3 = RigVMDispatchNode(Name=Global Right Control Color, NameSpace=Self, Value=RigVMVariableNode() {variable=Global Right Control Color}.Value) {template="DISPATCH_RigDispatch_SetModuleMetadata(in Name,in NameSpace,in Value,out Success)"}
   > Set_Module_Metadata_2_1_1_1_1_1 = RigVMDispatchNode(Name=Global Center Control Color, NameSpace=Self, Value=RigVMVariableNode() {variable=Global Center Control Color}.Value) {template="DISPATCH_RigDispatch_SetModuleMetadata(in Name,in NameSpace,in Value,out Success)"}
   Make_Array = RigVMDispatchNode(
    Values=((Type=None,Name="None"),((),Type=None,Name="None"))
    Values.0=RigVMVariableNode() {variable=Root Bone}.Value
    Values.1=RigVMVariableNode() {variable=Body}.Value
   ) {template="DISPATCH_RigVMDispatch_ArrayMake(in Values,out Array)"}
   Get Chain Length = RigVMFunctionReferenceNode(Initial=true, Items=@Make_Array.Array, ExecuteContext=@RigVMFunction_Sequence.D)
   > Set Control Scale = RigVMFunctionReferenceNode(
    Auto Scale Factor=&RigVMFunction_MathDoubleDiv(B=96, A=@"Get Chain Length".Length).Result
    Controls=&RigUnit_ItemArray(
     Items=((Type=None,Name="None"),(Type=None,Name="None"),(Type=None,Name="None"))
     Items.1=RigVMVariableNode() {variable=Local Control}.Value
     Items.2=RigVMVariableNode() {variable=Root Control}.Value
     Items.0=RigVMVariableNode() {variable=Global Control}.Value
    ).Items
    Scale=RigVMVariableNode() {variable=Control Scale}.Value
    Global Scale=RigVMVariableNode() {variable=Global Control Scale}.Value
   )
  ]}
 New Function_ContainedGraph {nodes [
   n1 = RigVMFunctionEntryNode()
   n2 = &RigUnit_GetTransform(bInitial=False, Item=@n1.Control)
   n3 = RigVMVariableNode() {variable=Distance}
   RigVMFunction_MathTransformTransformVector = /Script/RigVM.RigVMFunction_MathTransformTransformVector(Location=(X=0,Y=0,Z=0), Transform=@n2.Transform, Location.Y=@n3.Value) {is=struct RigVMUnitNode}
   Minimum = /Script/RigVM.RigVMFunction_MathDoubleMin(B=1, A=@n3.Value) {is=struct RigVMUnitNode, template="Minimum::Execute(in A,in B,out Result)"}
   > RigVMFunction_DebugLineStripNoSpace = /Script/RigVM.RigVMFunction_DebugLineStripNoSpace(
    DebugDrawSettings=(DepthPriority=SDPG_Foreground,Lifetime=-1)
    Points=((X=0,Y=0,Z=0),(X=0,Y=300,Z=0))
    WorldOffset=(Rotation=(X=0,Y=0,Z=0,W=1),Translation=(X=0,Y=0,Z=0),Scale3D=(X=1,Y=1,Z=1))
    Points.0=@n2.Transform.Translation
    Points.1=@RigVMFunction_MathTransformTransformVector.Result
    Thickness=@Minimum.Result
    bEnabled=RigVMVariableNode() {variable=Enabled}.Value
    Color=/Script/ControlRig.RigUnit_HierarchyGetShapeSettings(Item=@n1.Control) {is=struct RigVMUnitNode}.Settings.Color
   ) {is=struct RigVMUnitNode}
   > RigVMFunctionReturnNode()
  ]}
 New Function_ContainedGraph {nodes [
   n1 = RigVMFunctionEntryNode()
   > n2 = /Script/ControlRig.RigUnit_HierarchyAddAnimationChannelBool(InitialValue=False, LimitsEnabled=(), MaximumValue=True, MinimumValue=False, Name=Control Path Vis, Parent=@n1.Control) {
    is=struct RigVMUnitNode
    template="SpawnAnimationChannel::Execute(in InitialValue,in MinimumValue,in MaximumValue,in LimitsEnabled,in Parent,in Name,out Item)"
   }
   > SpawnAnimationChannel = /Script/ControlRig.RigUnit_HierarchyAddAnimationChannelFloat(
    InitialValue=50
    LimitsEnabled=(Enabled=(bMinimum=True,bMaximum=false))
    MaximumValue=500
    MinimumValue=0
    Name=Control Path Distance
    Parent=RigVMVariableNode() {variable=Control}.Value
   ) {
    is=struct RigVMUnitNode
    template="SpawnAnimationChannel::Execute(in InitialValue,in MinimumValue,in MaximumValue,in LimitsEnabled,in Parent,in Name,out Item)"
   }
   > RigVMFunctionReturnNode(Distance Channel=@SpawnAnimationChannel.Item, Vis Channel=@n2.Item)
  ]}
 RigVMFunctionLibrary {nodes [
   Generate Control Path = RigVMCollapseNode(Control=(Type=Control,Name="None"), ExecuteContext=())
   Set Up Control Path = RigVMCollapseNode(Control=(Type=Bone,Name="None"), ExecuteContext=())
  ]}
 RigVMFunction_Sequence_1_ContainedGraph {nodes [
   n1 = RigVMFunctionEntryNode()
   RigVMFunction_Sequence = &RigVMFunction_Sequence(ExecuteContext=@n1.ExecuteContext)
   n2 = &RigVMFunction_Sequence(ExecuteContext=@RigVMFunction_Sequence.B)
   n3 = &RigVMFunction_Sequence(ExecuteContext=@n2.B)
   RigVMFunctionReturnNode(A=@RigVMFunction_Sequence.A, B=@n2.A, C=@n3.A, D=@n3.B)
  ]}
 RigVMFunction_Sequence_1_ContainedGraph {nodes [
   n1 = RigVMFunctionEntryNode()
   RigVMFunction_Sequence = &RigVMFunction_Sequence(ExecuteContext=@n1.ExecuteContext)
   n2 = &RigVMFunction_Sequence(ExecuteContext=@RigVMFunction_Sequence.B)
   n3 = &RigVMFunction_Sequence(ExecuteContext=@n2.B)
   RigVMFunctionReturnNode(A=@RigVMFunction_Sequence.A, B=@n2.A, C=@n3.A, D=@n3.B)
  ]}
 RigVMModel {nodes [
   RigUnit_BeginExecution = &RigUnit_BeginExecution()
   RigVMFunction_Sequence = RigVMAggregateNode(ExecuteContext=@RigUnit_BeginExecution.ExecutePin)
   GetAnimationChannelFromItem = /Script/ControlRig.RigUnit_GetFloatAnimationChannelFromItem(bInitial=False, Item=RigVMVariableNode() {variable=Global Control Path Distance Channel}.Value) {is=struct RigVMUnitNode, template="GetAnimationChannelFromItem::Execute(out Value,in Item,in bInitial)"}
   Generate Control Path = RigVMFunctionReferenceNode(
    Distance=@GetAnimationChannelFromItem.Value
    Enabled=/Script/ControlRig.RigUnit_GetBoolAnimationChannelFromItem(bInitial=False, Item=RigVMVariableNode() {variable=Global Control Path Vis Channel}.Value) {is=struct RigVMUnitNode, template="GetAnimationChannelFromItem::Execute(out Value,in Item,in bInitial)"}.Value
    ExecuteContext=@RigVMFunction_Sequence.C
    Control=RigVMVariableNode() {variable=Global Control}.Value
   )
   > Generate Control Path_1 = RigVMFunctionReferenceNode(
    Distance=&RigUnit_GetFloatAnimationChannelFromItem(bInitial=False, Item=RigVMVariableNode() {variable=Local Control Path Distance Channel}.Value) {template="GetAnimationChannelFromItem::Execute(out Value,in Item,in bInitial)"}.Value
    Enabled=&RigUnit_GetBoolAnimationChannelFromItem(bInitial=False, Item=RigVMVariableNode() {variable=Local Control Path Vis Channel}.Value) {template="GetAnimationChannelFromItem::Execute(out Value,in Item,in bInitial)"}.Value
    Control=RigVMVariableNode() {variable=Local Control}.Value
   )
   &RigUnit_SetTransform(
    Weight=1
    bInitial=False
    Value=&RigUnit_GetTransform(bInitial=False, Item=RigVMVariableNode() {variable=Root Control}.Value).Transform
    Item=RigVMVariableNode() {variable=Root Bone}.Value
    ExecutePin=@RigVMFunction_Sequence.B
   )
   n1 = &RigVMFunction_ControlFlowBranch(Condition=RigVMVariableNode() {variable=Lock Scale}.Value, ExecuteContext=@RigVMFunction_Sequence.A)
   Make_Array = RigVMDispatchNode(
    Values=((Type=None,Name="None"),(Type=None,Name="None"))
    Values.1=RigVMVariableNode() {variable=Local Control}.Value
    Values.0=RigVMVariableNode() {variable=Global Control}.Value
   ) {template="DISPATCH_RigVMDispatch_ArrayMake(in Values,out Array)"}
   For_Each = RigVMDispatchNode(Array=@Make_Array.Array, ExecuteContext=@n1.True) {template="DISPATCH_RigVMDispatch_ArrayIterator(in Array,out Element,out Index,out Count,out Ratio)"}
   > /Script/ControlRig.RigUnit_SetScale(Scale=(X=1,Y=1,Z=1), Space=LocalSpace, Weight=1, bInitial=False, bPropagateToChildren=True, Item=@For_Each.Element) {is=struct RigVMUnitNode}
  ]}
 RigVMModel Backwards Solve Graph {nodes [
   n1 = /Script/ControlRig.RigUnit_InverseExecution() {is=struct RigVMUnitNode}
   RigVMFunction_Sequence = &RigVMFunction_Sequence(ExecuteContext=@n1.ExecutePin)
   For_Each = RigVMDispatchNode(
    Array=&RigUnit_ItemArray(
     Items=((Type=None,Name="None"),(Type=None,Name="None"),(Type=None,Name="None"))
     Items.0=RigVMVariableNode() {variable=Global Control}.Value
     Items.1=RigVMVariableNode() {variable=Local Control}.Value
     Items.2=RigVMVariableNode() {variable=Root Control}.Value
    ).Items
    ExecuteContext=@RigVMFunction_Sequence.A
   ) {template="DISPATCH_RigVMDispatch_ArrayIterator(in Array,out Element,out Index,out Count,out Ratio)"}
   > &RigUnit_SetTransform(
    Weight=1
    bInitial=false
    bPropagateToChildren=true
    Item=@For_Each.Element
    Value=&RigUnit_GetTransform(bInitial=false, Item=RigVMVariableNode() {variable=Root Bone}.Value).Transform
   )
  ]}
]
hierarchy [
 RootJoint {type=Connector, loc=<0 0 0>, rot=<0 0 0 1>}
 root {type=Socket, loc=<0 0 0>, rot=<0 0 0 1>}
 Body {type=Connector, loc=<0 0 0>, rot=<0 0 0 1>}
]
variables [
 Body {type=FRigElementKey, default=(Type=Connector,Name="")}
 Control Scale {type=double, default=1}
 Global Center Control Color {type=FLinearColor, default=(R=1,G=1,B=0,A=1)}
 Global Control {type=FRigElementKey, default=(Type=Bone,Name="")}
 Global Control Path Distance Channel {type=FRigElementKey, default=(Type=Bone,Name="")}
 Global Control Path Vis Channel {type=FRigElementKey, default=(Type=Bone,Name="")}
 Global Control Scale {type=double, default=1}
 Global Left Control Color {type=FLinearColor, default=(R=0,G=0,B=1,A=1)}
 Global Right Control Color {type=FLinearColor, default=(R=1,G=0,B=0,A=1)}
 Local Control {type=FRigElementKey, default=(Type=Bone,Name="")}
 Local Control Path Distance Channel {type=FRigElementKey, default=(Type=Bone,Name="")}
 Local Control Path Vis Channel {type=FRigElementKey, default=(Type=Bone,Name="")}
 Lock Scale {type=bool, default=False}
 Root Bone {type=FRigElementKey, default=(Type=Connector,Name="")}
 Root Control {type=FRigElementKey, default=(Type=Bone,Name="")}
]
